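# Idle Niagara effect for the Clash backpack cosmetic, layering mesh tendrils and socket-following energy ribbons.
# MeshTendrils attaches meshes to skeletal bones or sockets by spawn group, pulsing scale and material parameters over life.
# Two BoneFlow emitters, named for stages 0-1 and 2-3, curl partitioned ribbons along skeletal splines; the small socket-ribbon emitter is disabled.
# User parameters drive spawn rates, tendril size, tint and emissive; stage 2-3 spawn rate defaults to zero.
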NiagaraSystem /BRCosmetics/Accessories/FORT_Backpacks/Backpack_M_MED_Clash/FX/NS_Backpack_Clash_Idle
dependencies {
 usedDataInterfaces [&NiagaraDataInterfaceCurve, &NiagaraDataInterfaceVectorCurve]
 usedDynamicInputs []
 usedModules [&SystemState]
}
emitters {
 MeshTendrils {
  emitterUpdateScript [
   &EmitterState(Life Cycle Mode=::NewEnumerator1, Loop Duration=1, Recalculate Duration Each Loop=true, UseLoopDelay=false)
   &SpawnRate(SpawnRate=$User.TendrilsSpawnRate, Use Spawn Probability=false)
  ]
  particleSpawnScript [
   &FN_SetSpawnGroup(Number of Spawn Groups=16)
   &InitializeParticle(
    Color=<rgba 0.276 0.276 0.276 1>
    Lifetime Min=1.5
    Lifetime Mode=::NewEnumerator1
    Mesh Scale=&Multiply_VectorByFloat(Vector=<1 0.8 1>, Float=&RandomRangeFloat(Minimum=0.5, Maximum=0.8))
    Mesh Scale Mode=::NewEnumerator4
    UsePositionOffset=false
    Write Material Random=true
   )
   &InitialMeshOrientation(Rotation=&RandomRangeVector(Minimum=<-0.14 -0.07 0>, Maximum=<0.14 0.07 0>), Use Rotation Vector=true) {module=InitialMeshOrientation001}
   &self:DivideParticles_02(HowManyRenderers=3) {module=DivideParticles_02}
  ]
  particleUpdateScript [
   &ParticleState()
   &ScaleMeshSize(Scale Factor=&VectorFromFloat(Value=&FloatFromCurve(FloatCurve=DataInterface() {propertyValues {Curve {keys [
          {arriveTangent=3.7291, interpMode=RCIM_Cubic, leaveTangent=3.7291, tangentMode=RCTM_User, time=0, value=0}
          <at 0.5 1 RCIM_Cubic>
          {arriveTangent=-2.7655, interpMode=RCIM_Cubic, leaveTangent=-2.7655, tangentMode=RCTM_User, time=1, value=0}
         ]}}}))) {enabled=false}
   &DynamicMaterialParameters(
    Index 0 Param 1=&FloatFromCurve(FloatCurve=DataInterface() {propertyValues {Curve {keys [<at 0 0 RCIM_Cubic>,<at 0.5 1 RCIM_Cubic>,<at 1 0 RCIM_Cubic>]}}}, Scale Curve=1.25)
    Index 0 Param 2=&FloatFromCurve(
     FloatCurve=DataInterface() {propertyValues {Curve {keys [<at 0 1 RCIM_Cubic>,<at 0.5 0.1 RCIM_Cubic>,<at 1 1 RCIM_Cubic>]}}}
     Scale Curve=3.5
    )
    Index 0 Param 4=1
    Module.Param0WriteEnabled=true
    Module.Param1WriteEnabled=true
    Module.Param2WriteEnabled=true
    Module.Param3WriteEnabled=true
   )
   &SkeletalMeshLocation(
    Apply Sampled Orientation to Mesh Particles=true
    Bone / Socket Index=$Particles.SpawnGroup
    Bone Sampling Mode=::NewEnumerator4
    Inherit Velocity=false
    MaskSpawnGroup=false
    Sampled Position Offset Coordinate Space:ENiagaraMeshTransforms=::NewEnumerator3
    Skeletal Mesh=$User.SkeletalMesh Tendrils
    Use Sampled Mesh Attribute as Ribbon ID=false
   )
   &ScaleMeshSize(Scale Factor=&VectorFromCurve(
     VectorCurve=DataInterface() {propertyValues {
       XCurve {keys [<at 0 0 RCIM_Cubic>,<at 0.5 1 RCIM_Cubic>,<at 1 0 RCIM_Cubic>]}
       YCurve {keys [<at 0 0 RCIM_Cubic>,<at 0.5 1 RCIM_Cubic>,<at 1 0 RCIM_Cubic>]}
       ZCurve {keys [
         {arriveTangent=2.4741, interpMode=RCIM_Cubic, leaveTangent=2.4741, tangentMode=RCTM_User, time=0, value=0}
         <at 0.5 1 RCIM_Cubic>
         {arriveTangent=-2.7109, interpMode=RCIM_Cubic, leaveTangent=-2.7109, tangentMode=RCTM_User, time=1, value=0}
        ]}
      }}
     Scale Curve=&VectorFromFloat(Value=0.85)
    )) {module=ScaleMeshSize001}
   &ScaleMeshSize(Scale Factor=&VectorFromFloat(Value=$User.TendrilsSizeScale)) {module=ScaleMeshSize002}
  ]
  rendererClasses [&NiagaraMeshRendererProperties]
  renderers [
   {rendererClass=&NiagaraMeshRendererProperties, rendererIndex=0}
   {rendererClass=&NiagaraMeshRendererProperties, rendererIndex=1}
   {rendererClass=&NiagaraMeshRendererProperties, rendererIndex=2}
  ]
 }
 Stage_0_1_BoneFlow {
  emitterSpawnScript [
   /Game/Effects/Niagara/Modules/Partition/FN_PartitionAttributes(Number of Bone Sets=1, Particles Per Partition=10, Partitions:NiagaraInt32=$User.BoneFlowSmallSpawn)
   /SetVariables() {isSetParameters=true}
  ]
  emitterUpdateScript [
   &EmitterState(Life Cycle Mode=::NewEnumerator1, Loop Duration=0.5, UseLoopDelay=false)
   &SpawnBurst_Instantaneous(Spawn Count=$Emitter.FN_PartitionAttributes.Spawn Total, Use Spawn Probability=false)
  ]
  particleSpawnScript [
   /Game/Effects/Niagara/Modules/Partition/FN_Particle_partition_v5/FN_Particle_Partition_v5(Enable Ribbon ID=true, Pattern:ENiagaraPartitionWaveForm=::NewEnumerator1, Seed:NiagaraInt32=$Emitter.Seed, Shuffle Partition ID=true)
   /Niagara/Modules/Spawn/Initialization/InitializeParticle(
    Lifetime=0.5
    Module.Write Color=true
    Module.Write Lifetime=false
    Module.Write Mass=true
    Module.Write Material Random=true
    Module.Write Position=false
    Module.Write Scale=false
    Module.Write SpriteRotation=false
    Module.Write SpriteSize=true
    Position=&SimulationPosition()
    Sprite Size=<5 5>
   ) {enabled=false}
   &InitializeRibbon(
    Color=$User.Color_Tint
    Lifetime=1.5
    Module.Write Color=true
    Module.Write Lifetime=true
    Module.Write Mass=true
    Module.Write Material Random=true
    Module.Write Position=false
    Module.Write RibbonTwist=false
    Module.Write RibbonWidth=true
    Ribbon Width=&RandomRangeFloat(Minimum=3, Maximum=7)
   )
   /Game/Effects/Niagara/Modules/BoneFlow/FN_Initialize_BoneFlow_Attributes_v2(
    Angle:NiagaraFloat=&Multiply_Float(
     A=$Particles.Partition.NormalizedID
     B=&UniformRangedFloat(Minimum=0, Maximum=0.5, Random Seed=$Particles.Partition.ID)
    )
    Radius:NiagaraFloat=&FloatFromCurve(
     FloatCurve=DataInterface() {propertyValues {Curve {keys [
         {arriveTangent=77401.0078, interpMode=RCIM_Cubic, leaveTangent=50.7651, tangentMode=RCTM_User, time=0, value=0}
         {arriveTangent=16.47, interpMode=RCIM_Cubic, leaveTangent=16.47, time=0.13, value=5.5786}
         <at 0.4924 8.1096 RCIM_Cubic>
         {arriveTangent=-15.9761, interpMode=RCIM_Cubic, leaveTangent=-15.9761, time=0.8801, value=5.5786}
         {arriveTangent=-40.3134, interpMode=RCIM_Cubic, leaveTangent=-1467.5861, tangentMode=RCTM_User, time=1, value=0}
        ]}}}
     CurveIndex=$Particles.Partition.NormalizedID
    )
    T:NiagaraFloat=$Particles.Partition.NormalizedID
   )
  ]
  particleUpdateScript [
   &ParticleState()
   &FN_BoneFlow(
    Angle:NiagaraFloat=&Add_Float(A=$Particles.Initial.FN_Initialize_BoneFlow_Attributes_v2.Angle, B=HlslExpression() {hlslExpression=Emitter.Age * .2})
    Height:NiagaraFloat=&Multiply_Float(A=$Particles.Partition.NormalizedID, B=1)
    Radius:NiagaraFloat=$Particles.Initial.FN_Initialize_BoneFlow_Attributes_v2.Radius
    Set Material Dynamic=false
    SkeletalMesh=DataInterface() {propertyValues {
      DefaultMesh=null
      ExcludeBoneName=root
      OptionalComponentClass=null
      PreviewMesh=/Game/Accessories/FORT_Backpacks/Backpack_M_MED_Clash/Meshes/M_MED_Clash_Pack
      SkinningMode=SkinOnTheFly
      SoftSourceActor=null
      SourceMode=Default
      UvSetIndex=0
      WholeMeshLOD=-1
      bExcludeBone=false
      bReadDeformedGeometry=true
      bRequireCurrentFrameData=true
      ComponentTags []
      FilteredBones []
      FilteredSockets [FX_Pack_Bottom, FX_Pack_Top, FX_Pack_Bottom2, FX_Pack_Center]
      MeshUserParameter {parameter {name=None, typeDefHandle {registeredTypeIndex=1}, varData []}}
      SamplingRegions []
     }}
    Use Sockets Instead=true
   )
   &ScaleRibbonWidth(Ribbon Width Scale=&Multiply_Float(
     A=&FloatFromCurve(
      FloatCurve=DataInterface() {propertyValues {Curve {keys [<at 0 0.7 RCIM_Cubic>,<at 0.2 1 RCIM_Cubic>,<at 0.5 0.7 RCIM_Cubic>]}}}
      Scale Curve=0.7
     )
     B=&FloatFromCurve(
      FloatCurve=DataInterface() {propertyValues {Curve {keys [<at 0 0.75 RCIM_Cubic>,<at 0.4995 1.2 RCIM_Cubic>,<at 1 0.75 RCIM_Cubic>]}}}
      CurveIndex=$Particles.Partition.NormalizedID
     )
    ))
   &DynamicMaterialParameters(
    Index 0 Param 1=&FloatFromCurve(FloatCurve=DataInterface() {propertyValues {Curve {keys [<at 0 0 RCIM_Cubic>,<at 0.5 1 RCIM_Cubic>,<at 1 0 RCIM_Cubic>]}}})
    Index 0 Param 3=1
    Index 0 Param 4=$User.EmissiveIntensity
    Module.Param0WriteEnabled=true
    Module.Param1WriteEnabled=false
    Module.Param2WriteEnabled=false
    Module.Param3WriteEnabled=true
   )
  ]
  rendererClasses [&NiagaraRibbonRendererProperties]
  renderers [{rendererClass=&NiagaraRibbonRendererProperties}]
 }
 Stage_0_1_Ribbon_Small {
  bEnabled=false
  emitterUpdateScript [
   &EmitterState(Life Cycle Mode=::NewEnumerator1, Loop Duration=1, UseLoopDelay=false)
   &FN_SpawnRibbon_v2(
    Number of Ribbons:NiagaraInt32=&Multiply_Int(A=4, B=&MakeCustomIntFromBool(Bool=$User.IsSkydiving, True INT=0, False INT=1))
    SpawnRate:NiagaraFloat=$User.RibbonSmallSpawnRate
   ) {module=FN_SpawnRibbon_v2001}
  ]
  particleSpawnScript [
   /Game/Effects/Niagara/Modules/Ribbons/FN_InitializeRibbon_v4(
    Enable Skeletal Position=true
    Lifetime=0.6
    Ribbon Width:NiagaraFloat=&RandomRangeFloat(Minimum=3, Maximum=6)
    SkeletalMesh=DataInterface() {propertyValues {
      DefaultMesh=null
      ExcludeBoneName=root
      OptionalComponentClass=null
      PreviewMesh=&M_MED_Clash_Pack
      SkinningMode=SkinOnTheFly
      SoftSourceActor=null
      SourceMode=Default
      UvSetIndex=0
      WholeMeshLOD=-1
      bExcludeBone=false
      bReadDeformedGeometry=true
      bRequireCurrentFrameData=true
      ComponentTags []
      FilteredBones []
      FilteredSockets [FX_Ribbon_01,FX_Ribbon_02,FX_Ribbon_03,FX_Ribbon_04]
      MeshUserParameter {parameter {name=None, typeDefHandle {registeredTypeIndex=1}, varData []}}
      SamplingRegions []
     }}
   )
   &AddVelocityFromPoint(Module.Use Falloff Distance=false, Velocity Strength=$User.RibbonSmallVelocity) {module=AddVelocityFromPoint001}
  ]
  particleUpdateScript [
   &InheritVelocity(
    Module.Limit Speed=true
    Velocity Limit=150
    Velocity Scale=&VectorFromFloat(Value=&FloatFromCurve(FloatCurve=DataInterface() {propertyValues {Curve {keys [<at 0 1>,<at 0.5 0>]}}}, Scale Curve=0.1))
   )
   /Niagara/Modules/Update/Forces/CurlNoiseForce(
    Noise Frequency=0.03
    Pan Noise=true
    Pan Noise Field=<0.05 0.05 0.05>
    Random Seed:NiagaraInt32=&RandomRangeInt(Minimum=0, Maximum=1)
   )
   &Drag(Drag=0.9, Use Linear Drag=true, Use Rotational Drag=false)
   &SolveForcesAndVelocity(Clamp Velocity=false, Limit Acceleration=false)
   /Game/Effects/Niagara/Modules/Ribbons/FN_UpdateRibbon_v4(
    Enable Skeletal Position=true
    SkeletalMesh=DataInterface() {propertyValues {
      DefaultMesh=null
      ExcludeBoneName=root
      OptionalComponentClass=null
      PreviewMesh=&M_MED_Clash_Pack
      SkinningMode=SkinOnTheFly
      SoftSourceActor=null
      SourceMode=Default
      UvSetIndex=0
      WholeMeshLOD=-1
      bExcludeBone=false
      bReadDeformedGeometry=true
      bRequireCurrentFrameData=true
      ComponentTags []
      FilteredBones []
      FilteredSockets [FX_Ribbon_01,FX_Ribbon_02,FX_Ribbon_03,FX_Ribbon_04]
      MeshUserParameter {parameter {name=None, typeDefHandle {registeredTypeIndex=1}, varData []}}
      SamplingRegions []
     }}
   )
   &DynamicMaterialParameters(
    Index 0 Param 1=&Lerp_Float(
     A=1
     B=-0.5
     Alpha=&NormalizeFloat(Value=&VectorLength(Vector=$Particles.Velocity), Min=0, Max=100)
    )
    Index 0 Param 2=$Particles.MaterialRandom
    Index 0 Param 3=1
    Index 0 Param 4=$User.EmissiveIntensity
    Module.Param0WriteEnabled=true
    Module.Param1WriteEnabled=true
    Module.Param2WriteEnabled=false
    Module.Param3WriteEnabled=true
   ) {module=DynamicMaterialParameters001}
   &ScaleRibbonWidth(Ribbon Width Scale=&FloatFromCurve(FloatCurve=DataInterface() {propertyValues {Curve {keys [
         <at 0 1 RCIM_Cubic>
         {arriveTangent=-1.6026, interpMode=RCIM_Cubic, leaveTangent=-1.6026, tangentMode=RCTM_User, time=0.7967, value=0.5043}
         {arriveTangent=-2.0105, interpMode=RCIM_Cubic, leaveTangent=-2.0105, tangentMode=RCTM_User, time=1, value=0}
        ]}}}))
  ]
  rendererClasses [&NiagaraRibbonRendererProperties]
  renderers [{rendererClass=&NiagaraRibbonRendererProperties}]
 }
 Stage_2_3_BoneFlow {
  emitterSpawnScript [
   /SetVariables(Emitter.RandomSeed:NiagaraInt32=/Niagara/DynamicInputs/Random/RandomInt(Randomness Mode:ENiagaraRandomnessMode=::NewEnumerator0, Random Seed:NiagaraInt32=0, Fixed Random Seed:NiagaraBool=false)) {isSetParameters=true}
   /SetVariables(Emitter.ParticlesPerGroup=10) {isSetParameters=true}
  ]
  emitterUpdateScript [
   /SetVariables(Emitter.NumberOfGroups=4) {isSetParameters=true}
   &EmitterState(Life Cycle Mode=::NewEnumerator1, Loop Duration=1, UseLoopDelay=false)
   &SpawnRate(SpawnRate=$User.BoneFlowSpawnRate, Use Spawn Probability=false)
  ]
  particleSpawnScript [
   /SetVariables(Particles.LoopCount:NiagaraInt32=$Emitter.LoopCount) {isSetParameters=true}
   &FN_SetSpawnGroup(
    Inclusivity=true
    Normalize Spawn Group=true
    Particles Per Group:NiagaraInt32=$Emitter.ParticlesPerGroup
    Pattern=::NewEnumerator0
    Set Ribbon ID and Ribbon Link Order=true
   )
   &InitializeRibbon(
    Lifetime=&RandomRangeFloat(Minimum=2, Maximum=2)
    Module.Write Color=true
    Module.Write Lifetime=true
    Module.Write Mass=true
    Module.Write Material Random=true
    Module.Write Position=false
    Module.Write RibbonTwist=false
    Module.Write RibbonWidth=true
    Ribbon Width=&RandomRangeFloat(Minimum=1, Maximum=3)
   )
   &FN_FlipRibbonV()
  ]
  particleUpdateScript [
   &ParticleState()
   /Game/Effects/Niagara/Modules/BoneFlow/FN_BoneFlow_v4/FN_BoneFlow(
    Angle:NiagaraFloat=&Add_Float(
     A=&Multiply_Float(A=$Particles.Age, B=0.5)
     B=&Lerp_Float(A=1, B=3, Alpha=$Particles.Normalized Spawn Group)
    )
    BoneID:NiagaraInt32=&ModuloInt(A=$Particles.SpawnGroup, B=4)
    Height:NiagaraFloat=&Multiply_Float(
     A=$Particles.Normalized Spawn Group
     B=&FloatFromCurve(FloatCurve=DataInterface() {propertyValues {Curve {keys [
          {arriveTangent=2.402, interpMode=RCIM_Cubic, leaveTangent=2.402, tangentMode=RCTM_User, time=0, value=0.6}
          {arriveTangent=-0.0795, interpMode=RCIM_Cubic, leaveTangent=-0.0795, tangentMode=RCTM_User, time=0.5, value=1}
          <at 1 0.65>
         ]}}})
    )
    Interpolate Mesh Sampling=false
    Radius:NiagaraFloat=&FloatFromCurve(
     FloatCurve=DataInterface() {propertyValues {Curve {keys [
         <at 0 0 RCIM_Cubic>
         {arriveTangent=0.0154, interpMode=RCIM_Cubic, leaveTangent=0.0154, tangentMode=RCTM_User, time=0.4, value=1}
         {arriveTangent=-0.3846, interpMode=RCIM_Cubic, leaveTangent=-0.3846, tangentMode=RCTM_User, time=1, value=0}
        ]}}}
     Scale Curve=&RandomRangeFloat(Minimum=8, Maximum=3)
    )
    SkeletalMesh:NiagaraDataInterfaceSkeletalMesh=$User.Skeletal Mesh Boneflow
    Stabalize Twisting=true
    Write Position Attribute=true
   ) {module=FN_BoneFlow001}
   /SetVariables(Particles.RibbonFacing=$Output.FN_BoneFlow001.SplineBinormal) {isSetParameters=true}
   &ScaleRibbonWidth(Ribbon Width Scale=&FloatFromCurve(
     FloatCurve=DataInterface() {propertyValues {Curve {keys [
         {arriveTangent=1.6417, interpMode=RCIM_Cubic, leaveTangent=1.6417, tangentMode=RCTM_User, time=0, value=0.4}
         {interpMode=RCIM_Cubic, tangentMode=RCTM_User, time=0.5, value=1}
         {arriveTangent=-4.5007, interpMode=RCIM_Cubic, leaveTangent=-4.5007, tangentMode=RCTM_User, time=1, value=0}
        ]}}}
     CurveIndex=$Particles.Normalized Spawn Group
     Scale Curve=&Multiply_Float(
      A=1
      B=&FloatFromCurve(
       FloatCurve=DataInterface() {propertyValues {Curve {keys [
           {arriveTangent=2.8989, interpMode=RCIM_Cubic, leaveTangent=2.8989, tangentMode=RCTM_User, time=0, value=0}
           {interpMode=RCIM_Cubic, tangentMode=RCTM_User, time=0.7, value=1}
           {arriveTangent=-2.6878, interpMode=RCIM_Cubic, leaveTangent=-2.6878, tangentMode=RCTM_User, time=1, value=0}
          ]}}}
       CurveIndex=$Particles.Age
      )
     )
    ))
   &DynamicMaterialParameters(
    Index 0 Param 1=&FloatFromCurve(FloatCurve=DataInterface() {propertyValues {Curve {keys [<at 0 0 RCIM_Cubic>,<at 0.5 1>,<at 1 0 RCIM_Cubic>]}}})
    Module.Param0WriteEnabled=false
    Module.Param1WriteEnabled=false
    Module.Param2WriteEnabled=false
    Module.Param3WriteEnabled=false
   )
   &SolveForcesAndVelocity(Clamp Velocity=false, Limit Acceleration=false)
  ]
  rendererClasses [&NiagaraRibbonRendererProperties]
  renderers [{rendererClass=&NiagaraRibbonRendererProperties}]
 }
}
summary {
 systemName=NS_Backpack_Clash_Idle
 emitters [
  {emitterName=MeshTendrils, rendererClasses [&NiagaraMeshRendererProperties]}
  {emitterName=Stage_2_3_BoneFlow, rendererClasses [&NiagaraRibbonRendererProperties]}
  {bEnabled=false, emitterName=Stage_0_1_Ribbon_Small, rendererClasses [&NiagaraRibbonRendererProperties]}
  {emitterName=Stage_0_1_BoneFlow, rendererClasses [&NiagaraRibbonRendererProperties]}
 ]
 userVariables [
  User.bisFrontEnd {type=NiagaraBool}
  User.bIsFrontEndPreview {type=NiagaraBool}
  User.BoneFlowSmallSpawn {type=NiagaraInt32, defaultValue=1}
  User.BoneFlowSpawnRate {type=NiagaraFloat, defaultValue=0, description=""}
  User.Color_Tint {type=LinearColor, description="", defaultValue {struct=LinearColor, value=<rgba 1 1 1 1>}}
  User.EmissiveIntensity {type=NiagaraFloat, defaultValue=0, description=""}
  User.IsSkydiving {type=NiagaraBool}
  User.RibbonIDOffset {type=NiagaraInt32}
  User.RibbonSmallSpawnRate {type=NiagaraFloat, defaultValue=10, description=""}
  User.RibbonSmallVelocity {type=NiagaraFloat, defaultValue=25, description=""}
  User.Skeletal Mesh Boneflow {
   type=NiagaraDataInterfaceSkeletalMesh
   defaultValue {
    struct=NiagaraExt_VariableValue_DataInterface
    value {dataInterface=&self:NiagaraDataInterfaceSkeletalMesh_1}
   }
  }
  User.SkeletalMesh {
   type=NiagaraDataInterfaceSkeletalMesh
   defaultValue {
    struct=NiagaraExt_VariableValue_DataInterface
    value {dataInterface=&self:NiagaraDataInterfaceSkeletalMesh_3}
   }
  }
  User.SkeletalMesh Tendrils {
   type=NiagaraDataInterfaceSkeletalMesh
   defaultValue {
    struct=NiagaraExt_VariableValue_DataInterface
    value {dataInterface=&self:NiagaraDataInterfaceSkeletalMesh_0}
   }
  }
  User.TendrilsSizeScale {type=NiagaraFloat, defaultValue=1, description=""}
  User.TendrilsSpawnRate {type=NiagaraFloat, defaultValue=5, description=""}
 ]
}
