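# Niagara transition effect for the Clash backpack cosmetic, combining bone-following ribbons with liquid tendril meshes on a skeletal mesh.
# BoneFlow bursts four ribbon groups, sized by a user count, along FN_BoneFlow splines, with age-driven twist, curve-shaped radius and width, and binormal facing.
# The tendril emitter bursts seven meshes that track sampled skeletal locations, get oriented and offset, split across three renderers, and scale over life and by User.TendrilsSize.
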NiagaraSystem /BRCosmetics/Accessories/FORT_Backpacks/Backpack_M_MED_Clash/FX/NS_Backpack_Clash_Transition
dependencies {
 usedDataInterfaces [&NiagaraDataInterfaceCurve, &NiagaraDataInterfaceSkeletalMesh]
 usedDynamicInputs []
 usedModules [
  &self:SystemSpawnScript.SystemScriptSource.SystemScriptGraph.NiagaraNodeAssignment_0.SetVariables_#1
  &SystemState
 ]
}
emitters {
 BoneFlow {
  emitterSpawnScript [
   /SetVariables(Emitter.RandomSeed:NiagaraInt32=/Niagara/DynamicInputs/Random/RandomInt(Randomness Mode:ENiagaraRandomnessMode=::NewEnumerator0, Random Seed:NiagaraInt32=0, Fixed Random Seed:NiagaraBool=false)) {isSetParameters=true}
   /SetVariables(Emitter.ParticlesPerGroup:NiagaraInt32=$User.BoneFlowPerGroup) {isSetParameters=true}
  ]
  emitterUpdateScript [
   /SetVariables(Emitter.NumberOfGroups=4) {isSetParameters=true}
   &EmitterState(Life Cycle Mode=::NewEnumerator1, Loop Behavior=::NewEnumerator1, Loop Duration=1, UseLoopDelay=false)
   &SpawnBurst_Instantaneous(
    Spawn Count=&Multiply_Int(A=$Emitter.ParticlesPerGroup, B=$Emitter.NumberOfGroups)
    Use Spawn Probability=false
   )
  ]
  particleSpawnScript [
   /SetVariables(Particles.LoopCount:NiagaraInt32=$Emitter.LoopCount) {isSetParameters=true}
   &FN_SetSpawnGroup(
    Inclusivity=true
    Normalize Spawn Group=true
    Particles Per Group:NiagaraInt32=$Emitter.ParticlesPerGroup
    Pattern=::NewEnumerator0
    Set Ribbon ID and Ribbon Link Order=true
   )
   &InitializeRibbon(
    Lifetime=4
    Module.Write Color=true
    Module.Write Lifetime=true
    Module.Write Mass=true
    Module.Write Material Random=true
    Module.Write Position=false
    Module.Write RibbonTwist=false
    Module.Write RibbonWidth=true
    Ribbon Width=&RandomRangeFloat(Minimum=3, Maximum=4)
   )
   &FN_FlipRibbonV()
  ]
  particleUpdateScript [
   &ParticleState()
   /Game/Effects/Niagara/Modules/BoneFlow/FN_BoneFlow_v4/FN_BoneFlow(
    Angle:NiagaraFloat=&Add_Float(
     A=&Multiply_Float(A=$Particles.Age, B=0.5)
     B=&Lerp_Float(A=0, B=1, Alpha=$Particles.Normalized Spawn Group)
    )
    BoneID:NiagaraInt32=&ModuloInt(A=$Particles.SpawnGroup, B=4)
    Height:NiagaraFloat=&Multiply_Float(
     A=$Particles.Normalized Spawn Group
     B=&FloatFromCurve(
      FloatCurve=DataInterface() {propertyValues {Curve {keys [
          {arriveTangent=2.2162, interpMode=RCIM_Cubic, leaveTangent=2.2162, tangentMode=RCTM_User, time=0, value=0}
          {arriveTangent=-0.017, interpMode=RCIM_Cubic, leaveTangent=-0.017, tangentMode=RCTM_User, time=0.8, value=1}
          {arriveTangent=-4.7911, interpMode=RCIM_Cubic, leaveTangent=-4.7911, tangentMode=RCTM_User, time=1, value=0.15}
         ]}}}
      Scale Curve=&RandomRangeFloat(Minimum=1, Maximum=1.5)
     )
    )
    Interpolate Mesh Sampling=false
    Radius:NiagaraFloat=&FloatFromCurve(
     FloatCurve=DataInterface() {propertyValues {Curve {keys [
         {arriveTangent=5.7764, interpMode=RCIM_Cubic, leaveTangent=5.7764, tangentMode=RCTM_User, time=0, value=0}
         {arriveTangent=-0.031, interpMode=RCIM_Cubic, leaveTangent=-0.031, tangentMode=RCTM_User, time=0.5, value=1}
         {arriveTangent=-0.3846, interpMode=RCIM_Cubic, leaveTangent=-0.3846, tangentMode=RCTM_User, time=1, value=0}
        ]}}}
     CurveIndex=$Particles.Normalized Spawn Group
     Scale Curve=&RandomRangeFloat(Minimum=25, Maximum=35)
    )
    SideVector1:Vector3f=<0 0 0>
    SideVector2:Vector3f=&RandomRangeVector(Minimum=<0 -1 0>, Maximum=<0 1 0>)
    SideVector3:Vector3f=<0 0 0>
    SkeletalMesh:NiagaraDataInterfaceSkeletalMesh=$System.Skeletal Mesh
    Stabalize Twisting=true
    Write Position Attribute=true
   ) {module=FN_BoneFlow001}
   /SetVariables(Particles.RibbonFacing=$Output.FN_BoneFlow001.SplineBinormal) {isSetParameters=true}
   &ScaleRibbonWidth(Ribbon Width Scale=&FloatFromCurve(
     FloatCurve=DataInterface() {propertyValues {Curve {keys [
         {arriveTangent=7.7351, interpMode=RCIM_Cubic, leaveTangent=7.7351, tangentMode=RCTM_User, time=0, value=0}
         {interpMode=RCIM_Cubic, tangentMode=RCTM_User, time=0.5, value=1}
         {arriveTangent=-6.7655, interpMode=RCIM_Cubic, leaveTangent=-6.7655, tangentMode=RCTM_User, time=1, value=0}
        ]}}}
     CurveIndex=$Particles.Normalized Spawn Group
     Scale Curve=&Multiply_Float(
      A=2.35
      B=&FloatFromCurve(
       FloatCurve=DataInterface() {propertyValues {Curve {keys [
           {arriveTangent=2.8989, interpMode=RCIM_Cubic, leaveTangent=2.8989, tangentMode=RCTM_User, time=0, value=0}
           {interpMode=RCIM_Cubic, tangentMode=RCTM_User, time=0.5, value=1}
           {arriveTangent=-2.6878, interpMode=RCIM_Cubic, leaveTangent=-2.6878, tangentMode=RCTM_User, time=1, value=0}
          ]}}}
       CurveIndex=$Particles.Age
      )
     )
    ))
   &DynamicMaterialParameters(
    Index 0 Param 1=&FloatFromCurve(FloatCurve=DataInterface() {propertyValues {Curve {keys [<at 0 0 RCIM_Cubic>,<at 0.5 1>,<at 1 0 RCIM_Cubic>]}}})
    Module.Param0WriteEnabled=false
    Module.Param1WriteEnabled=false
    Module.Param2WriteEnabled=false
    Module.Param3WriteEnabled=false
   )
  ]
  rendererClasses [&NiagaraRibbonRendererProperties]
  renderers [{rendererClass=&NiagaraRibbonRendererProperties}]
 }
 Mesh_Liquid_Tendril_01 {
  emitterUpdateScript [
   &EmitterState()
   &SpawnBurst_Instantaneous(Spawn Count=7, Use Spawn Probability=false)
  ]
  particleSpawnScript [
   /Niagara/Modules/Spawn/Initialization/InitializeParticle(
    Lifetime=&RandomRangeFloat(Minimum=1.3, Maximum=1.5)
    Mesh Scale=&Multiply_VectorByFloat(Vector=<1.1 1.1 1.1>, Float=&RandomRangeFloat(Minimum=0.9, Maximum=1))
    Module.Write Color=true
    Module.Write Lifetime=true
    Module.Write Mass=true
    Module.Write Material Random=true
    Module.Write Position=false
    Module.Write Scale=true
    Module.Write SpriteRotation=false
    Module.Write SpriteSize=false
    Position=&SimulationPosition()
   )
   &DynamicMaterialParameters(
    Index 0 Param 2=2
    Module.Param0WriteEnabled=false
    Module.Param1WriteEnabled=true
    Module.Param2WriteEnabled=false
    Module.Param3WriteEnabled=true
   )
   &self:DivideParticles(HowManyRenderers=3) {module=DivideParticles}
  ]
  particleUpdateScript [
   &SolveForcesAndVelocity(Clamp Velocity=false, Limit Acceleration=false)
   &SkeletalMeshLocation(
    Apply Sampled Orientation to Mesh Particles=true
    Inherit Velocity=false
    MaskSpawnGroup=false
    Mesh Sampling Type=::NewEnumerator5
    Skeletal Mesh=$System.Skeletal Mesh
    Socket Sampling Mode=::NewEnumerator5
   )
   &self:OrientAndOffsetParticles(Angle:NiagaraFloat=&RandomRangeFloat(Minimum=-0.5, Maximum=1)) {module=OrientAndOffsetParticles}
   &ParticleState()
   &ScaleMeshSize(Scale Factor=&VectorFromFloat(Value=&FloatFromCurve(FloatCurve=DataInterface() {propertyValues {Curve {keys [
          {arriveTangent=6.9385, interpMode=RCIM_Cubic, time=0, value=0}
          {arriveTangent=0.4411, interpMode=RCIM_Cubic, leaveTangent=0.4411, tangentMode=RCTM_User, time=0.2, value=1}
          {arriveTangent=-0.7441, interpMode=RCIM_Cubic, leaveTangent=-0.7441, tangentMode=RCTM_User, time=0.7, value=0.8}
          {interpMode=RCIM_Cubic, leaveTangent=-3.8739, time=1, value=0}
         ]}}})))
   &ScaleMeshSize(Scale Factor=&VectorFromFloat(Value=$User.TendrilsSize)) {module=ScaleMeshSize001}
  ]
  rendererClasses [&NiagaraMeshRendererProperties]
  renderers [
   {rendererClass=&NiagaraMeshRendererProperties, rendererIndex=0}
   {rendererClass=&NiagaraMeshRendererProperties, rendererIndex=1}
   {rendererClass=&NiagaraMeshRendererProperties, rendererIndex=2}
  ]
 }
}
summary {
 systemName=NS_Backpack_Clash_Transition
 emitters [
  {emitterName=Mesh_Liquid_Tendril_01, rendererClasses [&NiagaraMeshRendererProperties]}
  {emitterName=BoneFlow, rendererClasses [&NiagaraRibbonRendererProperties]}
 ]
 userVariables [
  User.BoneFlowPerGroup {type=NiagaraInt32}
  User.RibbonIDOffset {type=NiagaraInt32}
  User.RiibbonSmallSpawnRate {type=NiagaraFloat, defaultValue=10, description=""}
  User.RiibbonSmallVelcity {type=NiagaraFloat, defaultValue=1, description=""}
  User.TendrilsSize {type=NiagaraFloat, defaultValue=1, description=""}
 ]
}
