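# Animation Blueprint for the Clash back bling, which transforms through four stages as the player's eliminations rise.
# On initialization and anim input events, it maps the AthenaKills cosmetic stat to a stage index.
# Events at one to five kills also pulse a flag, cleared after evaluation, that plays the transform into that stage before its idle.
# A frontend reactivity bool selects between this and a state machine auto-advancing through every stage.
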AnimBlueprint /BRCosmetics/Accessories/FORT_Backpacks/Backpack_M_MED_Clash/Meshes/M_MED_Clash_Pack_AnimBP
target_skeleton=/Game/Accessories/FORT_Backpacks/Backpack_M_MED_Clash/Meshes/M_MED_Clash_Pack_Skeleton
graphs [
 AnimGraph {nodes [
   AnimGraphNode_StateMachine_0 = Front End State Machine() {is=title AnimGraphNode_StateMachine}
   AnimGraphNode_StateMachine_1 = Cosmetic State Machine() {is=title AnimGraphNode_StateMachine}
   Output Pose(Result=Blend Poses by bool(
     BlendPose_0=@AnimGraphNode_StateMachine_0
     BlendPose_1=@AnimGraphNode_StateMachine_1
     bActiveValue=Get bShouldDisplayFrontendReactivity() -> (bShouldDisplayFrontendReactivity:bool)
    ) {blendType=Linear}) {layerGroup=None}
  ]}
 Front End State Machine {
  entry=Stage 1
  path=AnimGraph.AnimGraphNode_StateMachine_0.Front End State Machine
  ids=id
  nodes [
   AnimStateNode_0 = Stage 1() {is=title AnimStateNode}
   AnimStateNode_1 = Transition 1 to 2() {is=title AnimStateNode}
   AnimStateNode_2 = Stage 2() {is=title AnimStateNode}
   AnimStateNode_3 = Transition 2 to 3() {is=title AnimStateNode}
   AnimStateNode_4 = Stage 3() {is=title AnimStateNode}
   AnimStateNode_5 = Transition 3 to 4() {is=title AnimStateNode}
   AnimStateNode_6 = "Stage  4"() {is=title AnimStateNode}
   AnimStateTransitionNode_0 = AnimStateTransitionNode() {bAutomaticRuleBasedOnSequencePlayerInState=true}
   AnimStateTransitionNode_1 = AnimStateTransitionNode() {bAutomaticRuleBasedOnSequencePlayerInState=true}
   AnimStateTransitionNode_2 = AnimStateTransitionNode() {bAutomaticRuleBasedOnSequencePlayerInState=true}
   AnimStateTransitionNode_3 = AnimStateTransitionNode() {bAutomaticRuleBasedOnSequencePlayerInState=true}
   AnimStateTransitionNode_4 = AnimStateTransitionNode() {bAutomaticRuleBasedOnSequencePlayerInState=true}
   AnimStateTransitionNode_5 = AnimStateTransitionNode() {bAutomaticRuleBasedOnSequencePlayerInState=true}
  ]
  links [
   {
    dst=Transition 1 to 2
    rule=AnimGraph.AnimGraphNode_StateMachine_0.Front End State Machine.AnimStateTransitionNode_0.Transition
    src=Stage 1
    via=AnimStateTransitionNode_0
   }
   {
    dst=Transition 2 to 3
    rule=AnimGraph.AnimGraphNode_StateMachine_0.Front End State Machine.AnimStateTransitionNode_2.Transition
    src=Stage 2
    via=AnimStateTransitionNode_2
   }
   {
    dst=Transition 3 to 4
    rule=AnimGraph.AnimGraphNode_StateMachine_0.Front End State Machine.AnimStateTransitionNode_4.Transition
    src=Stage 3
    via=AnimStateTransitionNode_4
   }
   {
    dst=Stage 2
    rule=AnimGraph.AnimGraphNode_StateMachine_0.Front End State Machine.AnimStateTransitionNode_1.Transition
    src=Transition 1 to 2
    via=AnimStateTransitionNode_1
   }
   {
    dst=Stage 3
    rule=AnimGraph.AnimGraphNode_StateMachine_0.Front End State Machine.AnimStateTransitionNode_3.Transition
    src=Transition 2 to 3
    via=AnimStateTransitionNode_3
   }
   {
    dst="Stage  4"
    rule=AnimGraph.AnimGraphNode_StateMachine_0.Front End State Machine.AnimStateTransitionNode_5.Transition
    src=Transition 3 to 4
    via=AnimStateTransitionNode_5
   }
  ]
 }
 Stage 1 {
  path=AnimGraph.AnimGraphNode_StateMachine_0.Front End State Machine.AnimStateNode_0.Stage 1
  nodes [Output Animation Pose(Result=Sequence Player 'Clash_Male_Stage01_Idle_BACKPACK'() {sequence=/BRCosmetics/Accessories/FORT_Backpacks/Backpack_M_MED_Clash/Animation/Clash_Male_Stage01_Idle_BACKPACK})]
 }
 Transition 1 to 2 {
  path=AnimGraph.AnimGraphNode_StateMachine_0.Front End State Machine.AnimStateNode_1.Transition 1 to 2
  nodes [Output Animation Pose(Result=Sequence Player 'Clash_Male_to_Stage02_BACKPACK'() {sequence=/BRCosmetics/Accessories/FORT_Backpacks/Backpack_M_MED_Clash/Animation/Clash_Male_to_Stage02_BACKPACK})]
 }
 Stage 2 {
  path=AnimGraph.AnimGraphNode_StateMachine_0.Front End State Machine.AnimStateNode_2.Stage 2
  nodes [Output Animation Pose(Result=Sequence Player 'Clash_Male_Stage02_Idle_BACKPACK'() {sequence=/BRCosmetics/Accessories/FORT_Backpacks/Backpack_M_MED_Clash/Animation/Clash_Male_Stage02_Idle_BACKPACK})]
 }
 Transition 2 to 3 {
  path=AnimGraph.AnimGraphNode_StateMachine_0.Front End State Machine.AnimStateNode_3.Transition 2 to 3
  nodes [Output Animation Pose(Result=Sequence Player 'Clash_Male_to_Stage03_BACKPACK'() {sequence=/BRCosmetics/Accessories/FORT_Backpacks/Backpack_M_MED_Clash/Animation/Clash_Male_to_Stage03_BACKPACK})]
 }
 Stage 3 {
  path=AnimGraph.AnimGraphNode_StateMachine_0.Front End State Machine.AnimStateNode_4.Stage 3
  nodes [Output Animation Pose(Result=Sequence Player 'Clash_Male_Stage03_Idle_BACKPACK'() {sequence=/Game/Accessories/FORT_Backpacks/Backpack_M_MED_Clash/Animation/Clash_Male_Stage03_Idle_BACKPACK})]
 }
 Transition 3 to 4 {
  path=AnimGraph.AnimGraphNode_StateMachine_0.Front End State Machine.AnimStateNode_5.Transition 3 to 4
  nodes [Output Animation Pose(Result=Sequence Player 'Clash_Male_to_Stage04_BACKPACK'() {sequence=/BRCosmetics/Accessories/FORT_Backpacks/Backpack_M_MED_Clash/Animation/Clash_Male_to_Stage04_BACKPACK})]
 }
 "Stage  4" {
  path="AnimGraph.AnimGraphNode_StateMachine_0.Front End State Machine.AnimStateNode_6.Stage  4"
  nodes [Output Animation Pose(Result=Sequence Player 'Clash_Male_Stage04_Idle_BACKPACK'() {sequence=/BRCosmetics/Accessories/FORT_Backpacks/Backpack_M_MED_Clash/Animation/Clash_Male_Stage04_Idle_BACKPACK})]
 }
 Transition {path=AnimGraph.AnimGraphNode_StateMachine_0.Front End State Machine.AnimStateTransitionNode_0.Transition, nodes [Result()]}
 Transition {path=AnimGraph.AnimGraphNode_StateMachine_0.Front End State Machine.AnimStateTransitionNode_1.Transition, nodes [Result()]}
 Transition {path=AnimGraph.AnimGraphNode_StateMachine_0.Front End State Machine.AnimStateTransitionNode_2.Transition, nodes [Result()]}
 Transition {path=AnimGraph.AnimGraphNode_StateMachine_0.Front End State Machine.AnimStateTransitionNode_3.Transition, nodes [Result()]}
 Transition {path=AnimGraph.AnimGraphNode_StateMachine_0.Front End State Machine.AnimStateTransitionNode_4.Transition, nodes [Result()]}
 Transition {path=AnimGraph.AnimGraphNode_StateMachine_0.Front End State Machine.AnimStateTransitionNode_5.Transition, nodes [Result()]}
 Cosmetic State Machine {
  entry=Idle
  path=AnimGraph.AnimGraphNode_StateMachine_1.Cosmetic State Machine
  ids=id
  nodes [
   AnimStateNode_0 = Idle()
   AnimStateNode_1 = on_Kill() {is=title AnimStateNode}
   AnimStateTransitionNode_0 = AnimStateTransitionNode()
   AnimStateTransitionNode_1 = AnimStateTransitionNode() {bAutomaticRuleBasedOnSequencePlayerInState=true}
  ]
  links [
   {
    dst=on_Kill
    rule=AnimGraph.AnimGraphNode_StateMachine_1.Cosmetic State Machine.AnimStateTransitionNode_0.Transition
    src=Idle
    via=AnimStateTransitionNode_0
   }
   {
    dst=Idle
    rule=AnimGraph.AnimGraphNode_StateMachine_1.Cosmetic State Machine.AnimStateTransitionNode_1.Transition
    src=on_Kill
    via=AnimStateTransitionNode_1
   }
  ]
 }
 Idle {
  path=AnimGraph.AnimGraphNode_StateMachine_1.Cosmetic State Machine.AnimStateNode_0.Idle
  nodes [
   n1 = Blend Poses by int!(
    BlendPose_0:Pose Link Structure=Sequence Player 'Clash_Male_Stage01_Idle_BACKPACK'() {sequence=&Clash_Male_Stage01_Idle_BACKPACK}
    BlendPose_1:Pose Link Structure=Sequence Player 'Clash_Male_Stage02_Idle_BACKPACK'() {sequence=&Clash_Male_Stage02_Idle_BACKPACK}
    BlendPose_2:Pose Link Structure=Sequence Player 'Clash_Male_Stage03_Idle_BACKPACK'() {sequence=&Clash_Male_Stage03_Idle_BACKPACK}
    BlendPose_3:Pose Link Structure=Sequence Player 'Clash_Male_Stage04_Idle_BACKPACK'() {sequence=&Clash_Male_Stage04_Idle_BACKPACK}
    BlendTime_0:float=0.1
    BlendTime_1:float=0.1
    BlendTime_2:float=0.1
    BlendTime_3:float=0.1
    ActiveChildIndex:int=Get Reactivity Stages() -> (Reactivity Stages:int=0)
   ) -> (Pose:Pose Link Structure) {
    is=title AnimGraphNode_BlendListByInt
    blendType=Linear
    blendPose [{linkId=-1, sourceLinkId=-1}, {linkId=-1, sourceLinkId=-1}, {linkId=-1, sourceLinkId=-1}, {linkId=-1, sourceLinkId=-1}]
    blendTime [0.1,0.1,0.1,0.1]
   }
   Output Animation Pose(Result=@n1)
  ]
 }
 on_Kill {
  path=AnimGraph.AnimGraphNode_StateMachine_1.Cosmetic State Machine.AnimStateNode_1.on_Kill
  nodes [
   n1 = Sequence Player 'Clash_Male_to_Stage02_BACKPACK'() {bLoopAnimation=false, sequence=&Clash_Male_to_Stage02_BACKPACK}
   n2 = Sequence Player 'Clash_Male_to_Stage03_BACKPACK'() {bLoopAnimation=false, sequence=&Clash_Male_to_Stage03_BACKPACK}
   n3 = Sequence Player 'Clash_Male_to_Stage04_BACKPACK'() {bLoopAnimation=false, sequence=&Clash_Male_to_Stage04_BACKPACK}
   n4 = Blend Poses by int(
    BlendPose_1=@n1
    BlendPose_2=@n2
    BlendPose_3=@n3
    BlendTime_0=0.1
    BlendTime_1=0.1
    BlendTime_2=0.1
    BlendTime_3=0.1
    ActiveChildIndex=Get Reactivity Stages() -> (Reactivity Stages=0)
   ) {
    blendType=Linear
    blendPose [{linkId=-1, sourceLinkId=-1}, {linkId=-1, sourceLinkId=-1}, {linkId=-1, sourceLinkId=-1}, {linkId=-1, sourceLinkId=-1}]
    blendTime [0.1,0.1,0.1,0.1]
   }
   Output Animation Pose(Result=@n4)
  ]
 }
 Transition {
  path=AnimGraph.AnimGraphNode_StateMachine_1.Cosmetic State Machine.AnimStateTransitionNode_0.Transition
  nodes [Result(bCanEnterTransition=Get bShould Transform Slate() -> (bShould Transform Slate:bool))]
 }
 Transition {path=AnimGraph.AnimGraphNode_StateMachine_1.Cosmetic State Machine.AnimStateTransitionNode_1.Transition, nodes [Result()]}
 EventGraph {nodes [
   n1 = Event BlueprintProcessAnimInputEvent()
   > n2 = GetCosmeticStatValue!(
    execute
    PlayerPawn:Fort Player Pawn Object Reference=Get As Fort Player Pawn() -> (As Fort Player Pawn:Fort Player Pawn Object Reference)
    StatTag:Fort Stat Manager Tag Structure=(TagName="StatManager.AthenaKills")
    DefaultValue:int=0
   ) -> (then, ReturnValue:int=0)
   > n3 = Set KillCount(KillCount:int=@n2) -> (Output_Get:int=0)
   > n4 = Switch on Int!(execute:Exec, Selection:int="Clamp (Integer)"(Value=@n3, Min=0, Max=6) -> (ReturnValue=0)) -> (Default:Exec, 0:Exec, 1:Exec, 2:Exec, 3:Exec, 4:Exec, 5:Exec, 6:Exec)
   n5 = Set Reactivity Stages(execute=[@n4.1, @n4.2], Reactivity Stages:int=1) -> (Output_Get:int=0)
   Set bShould Transform Slate(execute=[@n7.then, @n5.then, @n6.then], bShould Transform Slate:bool=true) -> (Output_Get:bool)
   n6 = Set Reactivity Stages(execute=[@n4.3, @n4.4], Reactivity Stages=2) -> (Output_Get=0)
   n7 = Set Reactivity Stages(execute=@n4.5, Reactivity Stages=3) -> (Output_Get=0)
   Event BlueprintPostEvaluateAnimation() {is=title K2Node_Event}
   > Set bShould Transform Slate(bShould Transform Slate=false)
   Event BlueprintInitializeAnimation()
   > Do Once()
   > n8 = Cast To FortPlayerPawn(Object=TryGetFortPawn())
   > n9 = Set As Fort Player Pawn(As Fort Player Pawn:Fort Player Pawn Object Reference=@n8) -> (Output_Get:Fort Player Pawn Object Reference)
   > n10 = GetCosmeticStatValue(PlayerPawn=@n9, StatTag=(TagName="StatManager.AthenaKills"), DefaultValue=0) -> (ReturnValue=0)
   > n11 = Set KillCount(KillCount=@n10) -> (Output_Get=0)
   > n12 = Switch on Int(Selection="Clamp (Integer)"(Value=@n11, Min=0, Max=5) -> (ReturnValue=0))
   Set Reactivity Stages(execute=@n12.0, Reactivity Stages=0) -> (Output_Get=0)
   Set Reactivity Stages(execute=[@n12.1, @n12.2], Reactivity Stages=1) -> (Output_Get=0)
   Set Reactivity Stages(execute=[@n12.3, @n12.4], Reactivity Stages=2) -> (Output_Get=0)
   Set Reactivity Stages(execute=@n12.5, Reactivity Stages=3) -> (Output_Get=0)
   TryGetPawnOwner()
   IsNamed(self=@n1.AnimInputEvent, InEventName=aie.EnemyPlayerEliminated)
   Event BlueprintUpdateAnimation()
   Get Elim Count() -> (Elim Count:int=0)
  ]}
]
variables [
 As Fort Player Pawn {type=FortPlayerPawn}
 bShould Transform Slate {type=boolean}
 Elim Count {type=integer}
 KillCount {type=integer}
 Reactivity Stages {type=integer}
]
